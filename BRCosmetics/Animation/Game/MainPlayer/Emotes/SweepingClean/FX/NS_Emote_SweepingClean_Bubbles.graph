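# Niagara effect for the SweepingClean player emote, emitting bubbles, mist, soap and suds around two sweeper sockets.
# Spawn rates respond to owner speed, and a front-end preview flag adjusts velocities, rates and colors.
# Bubbles record death positions through a custom WriteDeath module, and system-level death variables trigger BurstStaticBubbles pop sprites.
# BlendSpace blends particles between world and local space by owner velocity, while time-of-day functions tune brightness and tint.
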NiagaraSystem /BRCosmetics/Animation/Game/MainPlayer/Emotes/SweepingClean/FX/NS_Emote_SweepingClean_Bubbles
dependencies {
 usedDataInterfaces [
  &NiagaraDataInterfaceArrayInt32
  &NiagaraDataInterfaceCurve
  &NiagaraDataInterfaceVectorCurve
  &NiagaraDataInterfaceColorCurve
  &NiagaraDataInterfaceArrayFloat
  &NiagaraDataInterfaceArrayColor
  &NiagaraDataInterfaceArrayPosition
  &NiagaraDataInterfaceSocketReader
 ]
 usedDynamicInputs []
 usedModules [
  &self:SystemSpawnScript.SystemScriptSource.SystemScriptGraph.NiagaraNodeAssignment_1.SetVariables_#1
  &SystemState
  &BlendSpaceHelper
  &self:SystemSpawnScript.SystemScriptSource.SystemScriptGraph.NiagaraNodeAssignment_0.SetVariables_#2
  &self:ReadDeath
  &self:SystemSpawnScript.SystemScriptSource.SystemScriptGraph.NiagaraNodeAssignment_2.SetVariables_#3
 ]
}
emitters {
 Bubbles {
  emitterUpdateScript [
   &EmitterState(Life Cycle Mode=::NewEnumerator1, Loop Delay=0.5, Loop Duration=1, UseLoopDelay=false)
   &SpawnRate(
    Spawn Probability=0.7
    SpawnRate=&MakeCustomFloatFromBool(
     Bool=&SetBoolByFloatComparison(A=&VectorLength(Vector=$User.OwnerLocalVelocity), B=60)
     True Float=20
     False Float=5
    )
    Use Spawn Probability=true
   )
  ]
  particleSpawnScript [
   &PartitionParticles()
   &InitializeParticle(
    Lifetime=&RandomRangeFloat(Minimum=1.3, Maximum=4.3)
    Sprite Size Mode=::NewEnumerator1
    Uniform Sprite Size Max=3
    Uniform Sprite Size Min=1
    UsePositionOffset=false
    Write Material Random=true
   )
   &ApplyInitialForces()
   &SocketLocation(
    SamplePosOffset=false
    Socket Index=&SelectIntFromArray(Direct Array Index=$Particles.PartitionParticles.Partition, Int Selection Array=DataInterface() {propertyValues {IntData [0,1]}})
    Socket Reader=$System.Sweeper Socket Reader
    Socket Sampling Mode=::NewEnumerator3
    VelocitySpeedLimit=false
   )
   &ShapeLocation(
    Handle Radius=0
    Large Radius=15
    MaskSpawnGroup=false
    Non Uniform Scale=<1 1 0>
    Offset=<0 0 -10>
    Offset Mode=::NewEnumerator0
    Override Seed=false
    Shape Origin=$Output.SocketLocation.SampledPosition
    Shape Primitive=::NewEnumerator3
    UseOwnerScale=true
   )
   &AddVelocity(
    Module.Use Falloff Distance=false
    Override Seed=false
    Velocity Mode=::NewEnumerator1
    Velocity Origin=$Output.SocketLocation.SampledPosition
    Velocity Speed=&RandomRangeFloat(Minimum=100, Maximum=120)
   )
   &AddVelocity(Velocity=&MakeCustomVectorFromBool(
     Bool=$User.bIsFrontEndPreview
     True Vector=&RandomRangeVector(Minimum=<0 -90 150>, Maximum=<0 -70 100>)
     False Vector=&RandomRangeVector(Minimum=<0 0 150>, Maximum=<0 0 100>)
    )) {module=AddVelocity001}
  ]
  particleUpdateScript [
   &ParticleState()
   &CurlNoiseForce(
    Noise Strength=&FloatFromCurve(
     FloatCurve=DataInterface() {propertyValues {Curve {keys [
         {interpMode=RCIM_Cubic, leaveTangentWeight=53.6425, tangentMode=RCTM_User, time=0, value=50}
         {arriveTangent=342.4859, interpMode=RCIM_Cubic, leaveTangent=342.4859, tangentMode=RCTM_User, time=1, value=150}
        ]}}}
     Scale Curve=&RandomRangeFloat(Minimum=5, Maximum=1)
    )
    Pan Noise=false
   )
   &GravityForce(Gravity=&RandomRangeVector(
     Minimum=&VectorFromCurve(VectorCurve=DataInterface() {propertyValues {
        XCurve {keys [<at 0 0>,<at 1 0>]}
        YCurve {keys [<at 0 0>,<at 1 0>]}
        ZCurve {keys [
          {interpMode=RCIM_Cubic, tangentMode=RCTM_User, time=0, value=10}
          {arriveTangent=88.4209, interpMode=RCIM_Cubic, leaveTangent=88.4209, tangentMode=RCTM_User, time=0.75, value=35}
         ]}
       }})
     Maximum=&VectorFromCurve(VectorCurve=DataInterface() {propertyValues {
        XCurve {keys [<at 0 0>,<at 1 0>]}
        YCurve {keys [<at 0 0>,<at 1 0>]}
        ZCurve {keys [
          {interpMode=RCIM_Cubic, tangentMode=RCTM_User, time=0, value=10}
          {arriveTangent=133.5137, interpMode=RCIM_Cubic, leaveTangent=133.5137, tangentMode=RCTM_User, time=0.75, value=70}
         ]}
       }})
    ))
   &DynamicMaterialParameters(
    Index 0 Param 1=&FloatFromCurve(FloatCurve=DataInterface() {propertyValues {Curve {keys [
         {arriveTangent=0.1706, interpMode=RCIM_Cubic, leaveTangent=0.1706, leaveTangentWeight=0.3081, tangentMode=RCTM_User, time=0, value=0}
         {arriveTangentWeight=0.3104, interpMode=RCIM_Cubic, leaveTangentWeight=0.0345, tangentMode=RCTM_User, time=0.9, value=0.2}
         {arriveTangent=3.2005, interpMode=RCIM_Cubic, leaveTangent=3.2005, tangentMode=RCTM_User, time=1, value=0.5}
        ]}}})
    Module.Index0Float0Write=true
    Module.Index0Float1Write=true
    Module.Index0Float2Write=true
    Module.Index0Float3Write=true
   )
   &ScaleColor(
    Scale Alpha=&FloatFromCurve(
     FloatCurve=DataInterface() {propertyValues {Curve {
        defaultValue=0.3
        keys [
         {tangentMode=RCTM_User, time=0, value=1}
         {arriveTangent=-0.0405, interpMode=RCIM_Cubic, leaveTangent=-0.0405, tangentMode=RCTM_User, time=0.7009, value=1}
         {arriveTangent=-8.0089, interpMode=RCIM_Cubic, leaveTangent=-8.0089, tangentMode=RCTM_User, time=1, value=0}
        ]
       }}}
     Scale Curve=&FNTimeOfDayScalar(Night Scale=0.1, Morning Scale=0.3, Day Scale=0.5, Evening Scale=0.3)
    )
    ScaleA=true
    ScaleRGB=true
   )
   &ParticleState() {module=ParticleState001}
   &ScaleSpriteSize(Scale Factor=&Vector2DFromFloat(Value=&FloatFromCurve(
      FloatCurve=DataInterface() {propertyValues {Curve {keys [
          {arriveTangent=4.1785, interpMode=RCIM_Cubic, leaveTangent=4.1785, tangentMode=RCTM_User, time=0, value=0}
          {arriveTangent=1.8588, interpMode=RCIM_Cubic, leaveTangent=1.8588, tangentMode=RCTM_User, time=0.1, value=0.35}
          {arriveTangent=-0.2915, interpMode=RCIM_Cubic, leaveTangent=-0.2915, tangentMode=RCTM_User, time=0.6, value=1}
          {arriveTangentWeight=0.199, interpMode=RCIM_Cubic, leaveTangentWeight=0.0479, tangentMode=RCTM_User, time=0.9719, value=1}
          <at 1 2 RCIM_Cubic>
         ]}}}
      Scale Curve=4
     )))
   &SpriteRotationRate(Rotation Rate=&RandomRangeFloat(Minimum=-40, Maximum=40))
   &Drag(
    Drag=&FloatFromCurve(
     FloatCurve=DataInterface() {propertyValues {Curve {
        defaultValue=1
        keys [
         {interpMode=RCIM_Cubic, tangentMode=RCTM_User, time=0, value=0.2547}
         {interpMode=RCIM_Cubic, tangentMode=RCTM_User, time=0.2421, value=0.8074}
        ]
       }}}
     Scale Curve=&RandomRangeFloat(Minimum=3, Maximum=5)
    )
    Use Linear Drag=true
    Use Rotational Drag=true
   )
   &SolveForcesAndVelocity(Clamp Velocity=false, Limit Acceleration=false)
   &BlendSpace("Position Blend (World <-> Local)":NiagaraFloat=&MakeCustomFloatFromBool(
     Bool=&SetBoolByFloatComparison(A=&VectorLength(Vector=$Engine.Owner.Velocity), B=200)
     True Float=1
     False Float=&FloatFromCurve(FloatCurve=DataInterface() {propertyValues {Curve {
         defaultValue=1
         keys [{interpMode=RCIM_Cubic, tangentMode=RCTM_User, time=0, value=0.6381}, {interpMode=RCIM_Cubic, tangentMode=RCTM_User, time=1, value=0.3685}]
        }}})
     Translation Contribution:NiagaraFloat=1
     Rotation Contribution:NiagaraFloat=&MakeCustomFloatFromBool(
      Bool=&SetBoolByFloatComparison(A=&VectorLength(Vector=$Engine.Owner.Velocity), B=200)
      True Float=1
      False Float=&FloatFromCurve(FloatCurve=DataInterface() {propertyValues {Curve {
          defaultValue=1
          keys [{interpMode=RCIM_Cubic, tangentMode=RCTM_User, time=0, value=0.4314}, {interpMode=RCIM_Cubic, tangentMode=RCTM_User, time=1, value=0}]
         }}})
     )
    ))
   &self:WriteDeath(Position Array:NiagaraDataInterfaceArrayPosition=$System.DeathArray) {module=WriteDeath}
  ]
  rendererClasses [&NiagaraSpriteRendererProperties]
  renderers [{rendererClass=&NiagaraSpriteRendererProperties}]
 }
 BurstStaticBubbles {
  emitterUpdateScript [
   &EmitterState(Life Cycle Mode=::NewEnumerator1, Loop Duration=1, UseLoopDelay=false)
   &SpawnPerFrame(Spawn:NiagaraBool=$System.DeathSpawn, Use Spawn Probability=false)
   &SpawnPerFrame(
    Spawn:NiagaraBool=$System.DeathSpawn
    Spawn Count:NiagaraInt32=&RandomRangeInt(Minimum=2, Maximum=3)
    Spawn Group=1
    Use Spawn Probability=false
   ) {module=SpawnPerFrame001}
  ]
  particleSpawnScript [
   /SetVariables(Particles.VisibilityTag=&SpawnGroup()) {isSetParameters=true}
   &InitializeParticle(
    Lifetime=0.6
    Mesh Scale Mode=::NewEnumerator3
    Position=&TransformPosition(Position=$System.DeathPosition, Source Space=::Local, Destination Space=::World)
    Position Mode=::NewEnumerator1
    Sprite Rotation Mode=::NewEnumerator2
    Sprite Size Mode=::NewEnumerator3
    Uniform Sprite Size=&RandomRangeFloat(Minimum=13, Maximum=16)
    UsePositionOffset=false
    Write Material Random=true
   )
   &DynamicMaterialParameters(
    Index 0 Param 1=5
    Index 0 Param 2=1
    Index 0 Param 3=1
    Index 0 Param 4=1
    Index 1 Param 1=6.667
    Index 1 Param 3=1
    Index 1 Param 4=1
    Module.Index0Float0Write=true
    Module.Index0Float1Write=true
    Module.Index0Float2Write=true
    Module.Index0Float3Write=true
    Module.Index1Float0Write=true
    Module.Index1Float1Write=true
    Module.Index1Float2Write=true
    Module.Index1Float3Write=true
    Write Parameter Index 1=::NewEnumerator1
   )
   /SetVariables(
    Particles.Drops Lifetime:NiagaraFloat=&RandomRangeFloat(Minimum=0.3, Maximum=0.2)
    Particles.Ring Lifetime=0.1
   ) {isSetParameters=true}
   &AddVelocity(
    Velocity=&RandomRangeVector(Minimum=<-15 -15 5>, Maximum=<15 15 21>)
    Velocity Speed Scale=&MaskFloatBySpawnGroup(Masked Float=0, Passthrough Float=1.5)
   )
   /SetVariables(Particles.Drops Size:Vector2f=&RandomRangeVector2D(Minimum=<0.5 2>, Maximum=<1 3>)) {isSetParameters=true}
  ]
  particleUpdateScript [
   &ParticleState()
   /SetVariables(
    Particles.Update Drops Age:NiagaraFloat=&Divide_Float(A=$Particles.Age, B=$Particles.Initial.Drops Lifetime)
    Particles.Update Ring Age:NiagaraFloat=&Divide_Float(A=$Particles.Age, B=$Particles.Initial.Ring Lifetime)
   ) {isSetParameters=true}
   /SetVariables(
    Particles.Update Drops Dynamic Parameters:Vector4f=<0 0 1 1>
    Particles.Update Drops Linear Color:LinearColor=&MakeLinearColorFromVectorAndFloat(
     "Vector (RGB)"=<1 1 1>
     "Float (Alpha)"=&FloatFromCurve(
      FloatCurve=DataInterface() {propertyValues {Curve {keys [<at 0.2917 1 RCIM_Cubic>,<at 1 0>]}}}
      CurveIndex=$Particles.Update Drops Age
      Scale Curve=&FNTimeOfDayScalar(Night Scale=0.75, Morning Scale=0.9, Evening Scale=0.8)
     )
    )
    Particles.Update Drops Size:Vector2f=&Multiply_Vector2D(
     A=$Particles.Initial.Drops Size
     B=&Vector2DFromFloat(Value=&FloatFromCurve(
       FloatCurve=DataInterface() {propertyValues {Curve {keys [
           {arriveTangent=1.3698, interpMode=RCIM_Cubic, leaveTangent=1.3698, tangentMode=RCTM_User, time=0, value=0.3}
           {tangentMode=RCTM_User, time=1, value=1}
          ]}}}
       CurveIndex=$Particles.Update Drops Age
      ))
    )
   ) {isSetParameters=true}
   &GravityForce(Gravity=&MaskVectorBySpawnGroup(Masked Vector=<0 0 0>, Passthrough Vector=&RandomRangeVector(Minimum=<0 0 -120>, Maximum=<0 0 -70>)))
   &ScaleSpriteSize(
    Scale Sprite Size Mode=::NewEnumerator0
    Uniform Scale Factor=&FloatFromCurve(
     FloatCurve=DataInterface() {propertyValues {Curve {keys [
         {arriveTangent=0.3236, interpMode=RCIM_Cubic, leaveTangent=0.3236, tangentMode=RCTM_User, time=0, value=0.7}
         {tangentMode=RCTM_User, time=1, value=1.2}
        ]}}}
     CurveIndex=$Particles.Update Ring Age
    )
   )
   &ScaleColor(
    Scale Alpha=&FloatFromCurve(
     FloatCurve=DataInterface() {propertyValues {Curve {keys [<at 0.2917 1 RCIM_Cubic>,<at 1 0>]}}}
     CurveIndex=$Particles.Update Ring Age
     Scale Curve=0.4
    )
    Scale RGB=&VectorFromFloat(Value=&FloatFromCurve(
      FloatCurve=DataInterface() {propertyValues {Curve {keys [
          <at 0 3 RCIM_Cubic>
          {arriveTangent=-0.9668, interpMode=RCIM_Cubic, leaveTangent=-0.9668, tangentMode=RCTM_User, time=0.3086, value=1.5}
          <at 1 1 RCIM_Cubic>
         ]}}}
      CurveIndex=$Particles.Update Ring Age
      Scale Curve=&FNTimeOfDayScalar(Night Scale=0.6, Morning Scale=0.9, Evening Scale=0.8)
     ))
    ScaleA=true
    ScaleRGB=true
   )
   &ScaleColor(
    Scale Alpha=&FNTimeOfDayScalar(Night Scale=0.8, Morning Scale=0.9, Evening Scale=0.9)
    Scale RGB=&MakeVectorFromLinearColorRGB(Color=&SelectLinearColorFromArray(
      Array Sampling Mode=::NewEnumerator1
      Direct Array Index=$Engine.QualityLevel
      Color Selection Array=DataInterface() {propertyValues {ColorData [<rgba 0 0 0 1>, <rgba 1 1 1 1>]}}
      Random Seed=?
     ))
    ScaleA=true
    ScaleRGB=true
   ) {module=ScaleColor001}
   &SolveForcesAndVelocity(Clamp Velocity=false, Limit Acceleration=false)
  ]
  rendererClasses [&NiagaraSpriteRendererProperties]
  renderers [
   {rendererClass=&NiagaraSpriteRendererProperties}
   {rendererClass=&NiagaraSpriteRendererProperties, rendererIndex=1}
  ]
 }
 Mist {
  emitterUpdateScript [
   &EmitterState(Life Cycle Mode=::NewEnumerator1, Loop Duration=1, UseLoopDelay=false)
   &SpawnRate(
    SpawnRate=&MakeCustomFloatFromBool(
     Bool=&SetBoolByFloatComparison(A=&VectorLength(Vector=$User.OwnerLocalVelocity), B=60)
     True Float=25
     False Float=&MakeCustomFloatFromBool(Bool=$User.bIsFrontEndPreview, True Float=25, False Float=10)
    )
    Use Spawn Probability=false
   )
  ]
  particleSpawnScript [
   &PartitionParticles()
   &InitializeParticle(
    Lifetime=&MakeCustomFloatFromBool(
     Bool=&SetBoolByFloatComparison(A=&VectorLength(Vector=$User.OwnerLocalVelocity), B=60)
     True Float=&RandomRangeFloat(Minimum=0.4, Maximum=0.7)
     False Float=&RandomRangeFloat(Minimum=1.5, Maximum=1)
    )
    Sprite Rotation Angle=&UniformRangedFloat(Minimum=0, Maximum=1)
    Sprite Rotation Mode=::NewEnumerator1
    Sprite Size Mode=::NewEnumerator3
    Uniform Sprite Size=&UniformRangedFloat(Minimum=15, Maximum=25)
    UsePositionOffset=false
    Write Material Random=true
   ) {module=InitializeParticle001}
   /SetVariables(Particles.InitSpriteRotRate=&UniformRangedFloat(Minimum=-1, Maximum=1)) {isSetParameters=true, module=SetVariables_#4}
   &ApplyInitialForces()
   &SocketLocation(
    SamplePosOffset=false
    Socket Index=&SelectIntFromArray(Direct Array Index=$Particles.PartitionParticles.Partition, Int Selection Array=DataInterface() {propertyValues {IntData [0,1]}})
    Socket Reader=$System.Sweeper Socket Reader
    Socket Sampling Mode=::NewEnumerator3
    VelocitySpeedLimit=false
   )
   &ShapeLocation(
    Handle Radius=0
    Large Radius=20
    MaskSpawnGroup=false
    Non Uniform Scale=<1 1 0>
    Offset=<0 0 -10>
    Offset Mode=::NewEnumerator0
    Override Seed=false
    Shape Origin=$Output.SocketLocation.SampledPosition
    Shape Primitive=::NewEnumerator3
    UseOwnerScale=true
   )
   &CameraOffset(Camera Offset Amount=&RandomRangeFloat(Minimum=10, Maximum=15))
   &AddVelocity(
    Module.Use Falloff Distance=false
    Origin Offset=<0 0 -20>
    Override Seed=false
    Velocity Mode=::NewEnumerator1
    Velocity Origin=$Output.SocketLocation.SampledPosition
    Velocity Speed=&RandomRangeFloat(Minimum=0, Maximum=20)
   )
  ]
  particleUpdateScript [
   &ParticleState()
   &SpriteRotationRate(Rotation Rate=&Multiply_Float(
     A=$Particles.InitSpriteRotRate
     B=&FloatFromCurve(
      FloatCurve=DataInterface() {propertyValues {Curve {keys [
          {arriveTangent=-1.9214, interpMode=RCIM_Cubic, leaveTangent=-1.9214, tangentMode=RCTM_User, time=-0.0046, value=1.0113}
          {arriveTangent=-0.2496, interpMode=RCIM_Cubic, leaveTangent=-0.2496, tangentMode=RCTM_User, time=0.9896, value=0.16}
         ]}}}
      CurveIndex=$Emitter.Age
     )
    ))
   &ScaleColor(
    Scale Alpha=&FloatFromCurve(FloatCurve=DataInterface() {propertyValues {Curve {keys [<at 0 0>,<at 0.2049 0.9765>,<at 0.7791 0.9107>,<at 1 0>]}}})
    Scale RGB=&MakeVectorFromLinearColorRGB(Color=&MakeCustomLinearColorFromBool(
      Bool=$User.bIsFrontEndPreview
      True Color=&ColorFromCurve(ColorCurve=DataInterface() {propertyValues {
         AlphaCurve {keys [<at 0 1 RCIM_Cubic>]}
         BlueCurve {keys [<at 0 2 RCIM_Cubic>,<at 0.5919 1.5 RCIM_Cubic>]}
         GreenCurve {keys [<at 0 1.5845 RCIM_Cubic>,<at 0.5919 1.0362 RCIM_Cubic>]}
         RedCurve {keys [<at 0 1.0845 RCIM_Cubic>,<at 0.5919 0.6536 RCIM_Cubic>]}
        }})
      False Color=&ColorFromCurve(ColorCurve=DataInterface() {propertyValues {
         AlphaCurve {keys [<at 0 1 RCIM_Cubic>]}
         BlueCurve {keys [<at 0 0.9068 RCIM_Cubic>,<at 0.5919 0.7679 RCIM_Cubic>]}
         GreenCurve {keys [<at 0 0.7185 RCIM_Cubic>,<at 0.5919 0.4656 RCIM_Cubic>]}
         RedCurve {keys [<at 0 0.4917 RCIM_Cubic>,<at 0.5919 0.2162 RCIM_Cubic>]}
        }})
     ))
    ScaleA=true
    ScaleRGB=true
   ) {module=ScaleColor001}
   &Drag(Drag=3, Use Linear Drag=true, Use Rotational Drag=false)
   &SolveForcesAndVelocity(Clamp Velocity=false, Limit Acceleration=false)
   &ScaleSpriteSize(Scale Factor=&Vector2DFromFloat(Value=&FloatFromCurve(FloatCurve=DataInterface() {propertyValues {Curve {keys [
          {arriveTangent=5.2354, interpMode=RCIM_Cubic, leaveTangent=5.2354, tangentMode=RCTM_User, time=-0.0046, value=0.25}
          <at 0.1439 1.4923>
          <at 1.0011 3.4809 RCIM_Cubic>
         ]}}})))
   &DynamicMaterialParameters(
    Index 0 Param 1=&FloatFromCurve(FloatCurve=DataInterface() {propertyValues {Curve {keys [<at 0 0 RCIM_Cubic>,<at 1.0011 0.2>]}}})
    Index 0 Param 2=&FloatFromCurve(FloatCurve=DataInterface() {propertyValues {Curve {keys [
         {arriveTangent=0.0453, interpMode=RCIM_Cubic, tangentMode=RCTM_Break, time=0, value=-1}
         {arriveTangent=0.0604, interpMode=RCIM_Cubic, leaveTangent=0.0604, tangentMode=RCTM_User, time=1, value=-1}
        ]}}})
    Index 0 Param 3=&FloatFromCurve(FloatCurve=DataInterface() {propertyValues {Curve {keys [{interpMode=RCIM_Cubic, leaveTangent=4.6779, time=0.2851, value=0}]}}})
    Index 1 Param 1=&FloatFromCurve(
     FloatCurve=DataInterface() {propertyValues {Curve {keys [<at 0 0>,<at 0.25 1>]}}}
     Scale Curve=&RandomRangeFloat(Minimum=2, Maximum=1)
    )
    Index 1 Param 2=&FloatFromCurve(
     FloatCurve=DataInterface() {propertyValues {Curve {keys [
         {interpMode=RCIM_Cubic, leaveTangent=3.7422, tangentMode=RCTM_Break, time=0, value=0.2}
         {interpMode=RCIM_Cubic, leaveTangent=0.9066, time=1, value=2}
        ]}}}
     Scale Curve=&RandomRangeFloat(Minimum=0.5, Maximum=1)
    )
    Index 1 Param 3=&FloatFromCurve(FloatCurve=DataInterface() {propertyValues {Curve {keys [
         {interpMode=RCIM_Cubic, leaveTangent=2.8445, tangentMode=RCTM_Break, time=0, value=0.2}
         {interpMode=RCIM_Cubic, leaveTangent=63.2978, time=0.25, value=1}
        ]}}})
    Index 1 Param 4=&FloatFromCurve(
     FloatCurve=DataInterface() {propertyValues {Curve {keys [<at 0 1>,<at 1 0.8>]}}}
     Scale Curve=&RandomRangeFloat(Minimum=0.5, Maximum=1)
    )
    Index 2 Param 1=&FloatFromCurve(FloatCurve=DataInterface() {propertyValues {Curve {keys [{interpMode=RCIM_Cubic, leaveTangent=3.2309, tangentMode=RCTM_Break, time=0.1, value=0}, <at 1 1>]}}})
    Index 2 Param 2=$Particles.MaterialRandom
    Index 2 Param 3=&FloatFromCurve(FloatCurve=DataInterface() {propertyValues {Curve {keys [
         {interpMode=RCIM_Cubic, tangentMode=RCTM_Break, time=0, value=1}
         {interpMode=RCIM_Cubic, leaveTangent=11.1233, tangentMode=RCTM_Break, time=1, value=2}
        ]}}})
    Index 2 Param 4=0.75
    Module.Param0WriteEnabled=true
    Module.Param0WriteEnabled1=true
    Module.Param0WriteEnabled2=true
    Module.Param1WriteEnabled=false
    Module.Param1WriteEnabled1=true
    Module.Param1WriteEnabled2=true
    Module.Param2WriteEnabled=true
    Module.Param2WriteEnabled1=true
    Module.Param2WriteEnabled2=true
    Module.Param3WriteEnabled=true
    Module.Param3WriteEnabled1=true
    Module.Param3WriteEnabled2=true
    Write Parameter Index 1=true
    Write Parameter Index 2=true
   )
   &BlendSpace("Position Blend (World <-> Local)":NiagaraFloat=&MakeCustomFloatFromBool(
     Bool=&SetBoolByFloatComparison(A=&VectorLength(Vector=$Engine.Owner.Velocity), B=500)
     True Float=1
     False Float=&FloatFromCurve(FloatCurve=DataInterface() {propertyValues {Curve {
         defaultValue=1
         keys [
          {arriveTangent=-1.1238, interpMode=RCIM_Cubic, leaveTangent=-1.1238, tangentMode=RCTM_User, time=0.2307, value=0.8034}
          {interpMode=RCIM_Cubic, tangentMode=RCTM_User, time=1, value=0.3685}
         ]
        }}})
     Translation Contribution:NiagaraFloat=1
     Rotation Contribution:NiagaraFloat=1
    ))
  ]
  rendererClasses [&NiagaraSpriteRendererProperties]
  renderers [{rendererClass=&NiagaraSpriteRendererProperties}]
 }
 Soap {
  emitterUpdateScript [
   &EmitterState(Life Cycle Mode=::NewEnumerator1, Loop Duration=1, UseLoopDelay=false)
   &SpawnRate(
    SpawnGroup=1
    SpawnRate=&MakeCustomFloatFromBool(
     Bool=&SetBoolByFloatComparison(A=&VectorLength(Vector=$User.OwnerLocalVelocity), B=60)
     True Float=60
     False Float=&MakeCustomFloatFromBool(Bool=$User.bIsFrontEndPreview, True Float=65, False Float=60)
    )
    Use Spawn Probability=false
   )
   &SpawnRate(
    SpawnRate=&MakeCustomFloatFromBool(
     Bool=&SetBoolByFloatComparison(A=&VectorLength(Vector=$User.OwnerLocalVelocity), B=60)
     True Float=85
     False Float=&MakeCustomFloatFromBool(Bool=$User.bIsFrontEndPreview, True Float=65, False Float=60)
    )
    Use Spawn Probability=false
   ) {module=SpawnRate001}
  ]
  particleSpawnScript [
   &PartitionParticles(Number Of Partitions=4)
   /SetVariables(Particles.VisibilityTag=&SpawnGroup()) {isSetParameters=true}
   &InitializeParticle(
    Lifetime=&MakeCustomFloatFromBool(
     Bool=&SetBoolByFloatComparison(A=&VectorLength(Vector=$User.OwnerLocalVelocity), B=60)
     True Float=&RandomRangeFloat(Minimum=0.9, Maximum=0.7)
     False Float=&MaskFloatBySpawnGroup(
      Masked Float=&RandomRangeFloat(Minimum=1.2, Maximum=1.4)
      Passthrough Float=&RandomRangeFloat(Minimum=1, Maximum=1.2)
     )
    )
    Sprite Rotation Angle=&UniformRangedFloat(Minimum=0, Maximum=1)
    Sprite Rotation Mode=::NewEnumerator1
    Sprite Size Mode=::NewEnumerator3
    Uniform Sprite Size=&MaskFloatBySpawnGroup(
     Masked Float=&UniformRangedFloat(Minimum=3, Maximum=16)
     Passthrough Float=&UniformRangedFloat(Minimum=7, Maximum=15)
    )
    UsePositionOffset=false
    Write Material Random=true
   ) {module=InitializeParticle001}
   /SetVariables(Particles.InitSpriteRotRate=&UniformRangedFloat(Minimum=-1, Maximum=1)) {isSetParameters=true, module=SetVariables_#4}
   &ApplyInitialForces()
   &SocketLocation(
    SamplePosOffset=false
    Socket Index=&SelectIntFromArray(
     Direct Array Index=$Particles.PartitionParticles.Partition
     Direct Array Mode=::NewEnumerator1
     Int Selection Array=DataInterface() {propertyValues {IntData [1,1,0,0]}}
    )
    Socket Reader=$System.Sweeper Socket Reader
    Socket Sampling Mode=::NewEnumerator3
    VelocitySpeedLimit=false
   )
   &ShapeLocation(
    Handle Radius=0
    Large Radius=&MaskFloatBySpawnGroup(Masked Float=25, Passthrough Float=17)
    MaskSpawnGroup=false
    Non Uniform Scale=<1 1 0>
    Offset=<0 0 -10>
    Offset Mode=::NewEnumerator0
    Override Seed=false
    Shape Origin=$Output.SocketLocation.SampledPosition
    Shape Primitive=::NewEnumerator3
    UseOwnerScale=true
   )
   &CameraOffset(Camera Offset Amount=&RandomRangeFloat(Minimum=10, Maximum=15))
   &AddVelocity(
    Module.Use Falloff Distance=false
    Origin Offset=<0 0 -20>
    Override Seed=false
    Velocity Mode=::NewEnumerator1
    Velocity Origin=$Output.SocketLocation.SampledPosition
    Velocity Speed=&RandomRangeFloat(Minimum=0, Maximum=20)
   )
  ]
  particleUpdateScript [
   &ParticleState()
   &SpriteRotationRate(Rotation Rate=&Multiply_Float(
     A=$Particles.InitSpriteRotRate
     B=&FloatFromCurve(
      FloatCurve=DataInterface() {propertyValues {Curve {keys [
          {arriveTangent=-1.9214, interpMode=RCIM_Cubic, leaveTangent=-1.9214, tangentMode=RCTM_User, time=-0.0046, value=1.0113}
          {arriveTangent=-0.2496, interpMode=RCIM_Cubic, leaveTangent=-0.2496, tangentMode=RCTM_User, time=0.9896, value=0.16}
         ]}}}
      CurveIndex=$Emitter.Age
      Scale Curve=300
     )
    ))
   &ScaleColor(
    Scale Alpha=&FloatFromCurve(
     FloatCurve=DataInterface() {propertyValues {Curve {keys [<at 0 0>,<at 0.2049 0.9765>,<at 0.7791 0.9107>,<at 1 0>]}}}
     Scale Curve=&MaskFloatBySpawnGroup(Masked Float=1, Passthrough Float=0.6)
    )
    Scale RGB=&MakeVectorFromLinearColorRGB(Color=&MaskLinearColorBySpawnGroup(
      Masked Linear Color=&MakeCustomLinearColorFromBool(
       Bool=$User.bIsFrontEndPreview
       True Color=&ColorFromCurve(ColorCurve=DataInterface() {propertyValues {
          AlphaCurve {keys [<at 0 1 RCIM_Cubic>]}
          BlueCurve {keys [<at 0 2 RCIM_Cubic>,<at 0.5919 1.5 RCIM_Cubic>]}
          GreenCurve {keys [<at 0 1.187 RCIM_Cubic>,<at 0.5919 0.8108 RCIM_Cubic>]}
          RedCurve {keys [<at 0 0.7326 RCIM_Cubic>,<at 0.5919 0.6189 RCIM_Cubic>]}
         }})
       False Color=&ColorFromCurve(ColorCurve=DataInterface() {propertyValues {
          AlphaCurve {keys [<at 0 1 RCIM_Cubic>]}
          BlueCurve {keys [<at 0 0.9068 RCIM_Cubic>,<at 0.5919 0.7679 RCIM_Cubic>]}
          GreenCurve {keys [<at 0 0.7185 RCIM_Cubic>,<at 0.5919 0.4656 RCIM_Cubic>]}
          RedCurve {keys [<at 0 0.4917 RCIM_Cubic>,<at 0.5919 0.2162 RCIM_Cubic>]}
         }})
      )
      Passthrough Linear Color=&FNTimeOfDayColor(
       Night Color=<rgba 0.1398 0.4195 1.5 1>
       Morning Color=&ColorFromCurve(ColorCurve=DataInterface() {propertyValues {
          AlphaCurve {keys [<at 0 1 RCIM_Cubic>,<at 1 0 RCIM_Cubic>]}
          BlueCurve {keys [<at 0 2 RCIM_Cubic>,<at 1 1.5 RCIM_Cubic>]}
          GreenCurve {keys [<at 0 1.5224 RCIM_Cubic>,<at 1 1.358 RCIM_Cubic>]}
          RedCurve {keys [<at 0 0.9363 RCIM_Cubic>,<at 1 1.0243 RCIM_Cubic>]}
         }})
       Day Color=&ColorFromCurve(ColorCurve=DataInterface() {propertyValues {
          AlphaCurve {keys [<at 0 1 RCIM_Cubic>,<at 1 0 RCIM_Cubic>]}
          BlueCurve {keys [<at 0 2 RCIM_Cubic>,<at 1 1.5 RCIM_Cubic>]}
          GreenCurve {keys [<at 0 1.5224 RCIM_Cubic>,<at 1 1.358 RCIM_Cubic>]}
          RedCurve {keys [<at 0 0.9363 RCIM_Cubic>,<at 1 1.0243 RCIM_Cubic>]}
         }})
       Evening Color=&ColorFromCurve(ColorCurve=DataInterface() {propertyValues {
          AlphaCurve {keys [<at 0 1 RCIM_Cubic>,<at 1 0 RCIM_Cubic>]}
          BlueCurve {keys [<at 0 2 RCIM_Cubic>,<at 1 1.5 RCIM_Cubic>]}
          GreenCurve {keys [<at 0 1.5224 RCIM_Cubic>,<at 1 1.358 RCIM_Cubic>]}
          RedCurve {keys [<at 0 0.9363 RCIM_Cubic>,<at 1 1.0243 RCIM_Cubic>]}
         }})
      )
     ))
    ScaleA=true
    ScaleRGB=true
   ) {module=ScaleColor001}
   &Drag(Drag=3, Use Linear Drag=true, Use Rotational Drag=false)
   &SolveForcesAndVelocity(Clamp Velocity=false, Limit Acceleration=false)
   &ScaleSpriteSize(Scale Factor=&Vector2DFromFloat(Value=&FloatFromCurve(
      FloatCurve=DataInterface() {propertyValues {Curve {keys [
          {arriveTangent=5.2354, interpMode=RCIM_Cubic, leaveTangent=5.2354, tangentMode=RCTM_User, time=-0.0046, value=0.25}
          <at 0.1439 1.4923>
          <at 1.0011 3.4809 RCIM_Cubic>
         ]}}}
      Scale Curve=&MaskFloatBySpawnGroup(Masked Float=1, Passthrough Float=1.5)
     )))
   &DynamicMaterialParameters(
    Index 0 Param 1=&FloatFromCurve(FloatCurve=DataInterface() {propertyValues {Curve {keys [<at 0 0 RCIM_Cubic>,<at 1.0011 0.2>]}}})
    Index 0 Param 2=&FloatFromCurve(FloatCurve=DataInterface() {propertyValues {Curve {keys [
         {arriveTangent=0.0453, interpMode=RCIM_Cubic, tangentMode=RCTM_Break, time=0, value=-1}
         {arriveTangent=0.0604, interpMode=RCIM_Cubic, leaveTangent=0.0604, tangentMode=RCTM_User, time=1, value=-1}
        ]}}})
    Index 0 Param 3=&FloatFromCurve(FloatCurve=DataInterface() {propertyValues {Curve {keys [{interpMode=RCIM_Cubic, leaveTangent=4.6779, time=0.2851, value=0}]}}})
    Index 1 Param 1=&FloatFromCurve(
     FloatCurve=DataInterface() {propertyValues {Curve {keys [<at 0 0>,<at 0.25 1>]}}}
     Scale Curve=&RandomRangeFloat(Minimum=2, Maximum=1)
    )
    Index 1 Param 2=&FloatFromCurve(
     FloatCurve=DataInterface() {propertyValues {Curve {keys [
         {interpMode=RCIM_Cubic, leaveTangent=3.7422, tangentMode=RCTM_Break, time=0, value=0.2}
         {interpMode=RCIM_Cubic, leaveTangent=0.9066, time=1, value=2}
        ]}}}
     Scale Curve=&RandomRangeFloat(Minimum=0.5, Maximum=1)
    )
    Index 1 Param 3=&FloatFromCurve(FloatCurve=DataInterface() {propertyValues {Curve {keys [
         {interpMode=RCIM_Cubic, leaveTangent=2.8445, tangentMode=RCTM_Break, time=0, value=0.2}
         {interpMode=RCIM_Cubic, leaveTangent=63.2978, time=0.25, value=1}
        ]}}})
    Index 1 Param 4=&FloatFromCurve(
     FloatCurve=DataInterface() {propertyValues {Curve {keys [<at 0 1>,<at 1 0.8>]}}}
     Scale Curve=&RandomRangeFloat(Minimum=0.5, Maximum=1)
    )
    Index 2 Param 1=&FloatFromCurve(FloatCurve=DataInterface() {propertyValues {Curve {keys [{interpMode=RCIM_Cubic, leaveTangent=3.2309, tangentMode=RCTM_Break, time=0.1, value=0}, <at 1 1>]}}})
    Index 2 Param 2=$Particles.MaterialRandom
    Index 2 Param 3=&FloatFromCurve(FloatCurve=DataInterface() {propertyValues {Curve {keys [
         {interpMode=RCIM_Cubic, tangentMode=RCTM_Break, time=0, value=1}
         {interpMode=RCIM_Cubic, leaveTangent=11.1233, tangentMode=RCTM_Break, time=1, value=2}
        ]}}})
    Index 2 Param 4=0.75
    Module.Index0Float0Write=true
    Module.Index0Float1Write=true
    Module.Index0Float2Write=true
    Module.Index0Float3Write=true
    Module.Index1Float0Write=true
    Module.Index1Float1Write=true
    Module.Index1Float2Write=true
    Module.Index1Float3Write=true
    Module.Index2Float0Write=true
    Module.Index2Float1Write=true
    Module.Index2Float2Write=true
    Module.Index2Float3Write=true
    Write Parameter Index 1=::NewEnumerator1
    Write Parameter Index 2=::NewEnumerator1
   )
   &BlendSpace("Position Blend (World <-> Local)":NiagaraFloat=&MakeCustomFloatFromBool(
     Bool=&SetBoolByFloatComparison(A=&VectorLength(Vector=$Engine.Owner.Velocity), B=300)
     True Float=1
     False Float=&FloatFromCurve(FloatCurve=DataInterface() {propertyValues {Curve {
         defaultValue=1
         keys [
          {arriveTangent=-0.7918, interpMode=RCIM_Cubic, leaveTangent=-0.7918, tangentMode=RCTM_User, time=0.2456, value=1}
          {arriveTangent=-0.1424, interpMode=RCIM_Cubic, leaveTangent=-0.1424, tangentMode=RCTM_User, time=1, value=0.6428}
         ]
        }}})
     Translation Contribution:NiagaraFloat=1
     Rotation Contribution:NiagaraFloat=1
    ))
   /SetVariables(Particles.Toon Dynamic Parameters:Vector4f=&MakeVector4(
     X=&MakeCustomFloatFromBool(
      Bool=&SetBoolByFloatComparison(A=&VectorLength(Vector=$User.OwnerLocalVelocity), B=60)
      True Float=0.2
      False Float=&FloatFromCurve(FloatCurve=DataInterface() {propertyValues {Curve {keys [<at 0 0.2 RCIM_Cubic>, {arriveTangentWeight=0.5418, interpMode=RCIM_Cubic, tangentMode=RCTM_User, time=1, value=0.2}]}}})
     )
     Y=&SelectFloatFromArray(
      Array Sampling Mode=::NewEnumerator0
      Direct Array Index=?
      Float Selection Array=DataInterface() {propertyValues {FloatData [0,1,2]}}
      Random Seed=0
     )
     Z=0
     W=0
    )) {isSetParameters=true}
  ]
  rendererClasses [&NiagaraSpriteRendererProperties]
  renderers [
   {rendererClass=&NiagaraSpriteRendererProperties}
   {rendererClass=&NiagaraSpriteRendererProperties, rendererIndex=1}
  ]
 }
 Suds {
  emitterUpdateScript [
   &EmitterState(Life Cycle Mode=::NewEnumerator1, Loop Duration=1, UseLoopDelay=false)
   &SpawnRate(
    Spawn Probability=0.7
    SpawnRate=&MakeCustomFloatFromBool(
     Bool=&SetBoolByFloatComparison(A=&VectorLength(Vector=$User.OwnerLocalVelocity), B=60)
     True Float=10
     False Float=&MakeCustomFloatFromBool(Bool=$User.bIsFrontEndPreview, True Float=10, False Float=0)
    )
    Use Spawn Probability=true
   )
   &SpawnRate(
    Spawn Probability=0.7
    SpawnGroup=1
    SpawnRate=&MakeCustomFloatFromBool(
     Bool=&SetBoolByFloatComparison(A=&VectorLength(Vector=$User.OwnerLocalVelocity), B=60)
     True Float=70
     False Float=&MakeCustomFloatFromBool(Bool=$User.bIsFrontEndPreview, True Float=70, False Float=0)
    )
    Use Spawn Probability=true
   ) {module=SpawnRate001}
  ]
  particleSpawnScript [
   /SetVariables(Particles.VisibilityTag=&SpawnGroup()) {isSetParameters=true}
   &PartitionParticles()
   &InitializeParticle(
    Lifetime=&SelectFloatFromArray(
     Array Sampling Mode=::NewEnumerator0
     Direct Array Index=?
     Float Selection Array=DataInterface() {propertyValues {FloatData [1.5,0.7,1,1.1,1.7,0.8,1.3,1.2,0.9]}}
     Random Seed=0
    )
    Position Offset=<0 0 -5>
    Sprite Rotation Angle=&UniformRangedFloat(Minimum=0, Maximum=1)
    Sprite Rotation Mode=::NewEnumerator1
    Sprite Size Mode=::NewEnumerator3
    Uniform Sprite Size=&MaskFloatBySpawnGroup(
     Masked Float=&UniformRangedFloat(Minimum=6, Maximum=13)
     Passthrough Float=&UniformRangedFloat(Minimum=20, Maximum=22)
    )
    UsePositionOffset=true
    Write Material Random=true
   ) {module=InitializeParticle001}
   /SetVariables(Particles.InitSpriteRotRate=&UniformRangedFloat(Minimum=-1, Maximum=1)) {isSetParameters=true, module=SetVariables_#4}
   &ApplyInitialForces()
   &SocketLocation(
    SamplePosOffset=false
    Socket Index=&SelectIntFromArray(Direct Array Index=$Particles.PartitionParticles.Partition, Int Selection Array=DataInterface() {propertyValues {IntData [0,1]}})
    Socket Reader=$System.Sweeper Socket Reader
    Socket Sampling Mode=::NewEnumerator3
    VelocitySpeedLimit=false
   )
   &ShapeLocation(
    Handle Radius=0
    Large Radius=20
    MaskSpawnGroup=false
    Non Uniform Scale=<1 1 0>
    Offset=&MaskVector(Mask Source=1, Mask Parameter=$Particles.PartitionParticles.Partition, Mask=0, Masked Vector=<-3 0 -15>, Passthrough Vector=<3 0 -15>)
    Offset Mode=::NewEnumerator0
    Override Seed=false
    Rotation Mode=::NewEnumerator2
    Shape Origin=$Output.SocketLocation.SampledPosition
    Shape Primitive=::NewEnumerator3
    U Distribution=0.5
    UseOwnerScale=true
    Yaw / Pitch / Roll=&MaskVector(Mask Source=1, Mask Parameter=$Particles.PartitionParticles.Partition, Mask=0, Masked Vector=<0.1 0 0>, Passthrough Vector=<-0.1 0 0>)
   ) {module=ShapeLocation001}
   &CameraOffset(Camera Offset Amount=&RandomRangeFloat(Minimum=10, Maximum=15))
   &AddVelocity(Velocity=&MakeCustomVectorFromBool(
     Bool=$User.bIsFrontEndPreview
     True Vector=&RandomRangeVector(Minimum=<0 -70 0>, Maximum=<0 -150 0>)
     False Vector=<0 0 0>
    ))
  ]
  particleUpdateScript [
   &ParticleState()
   &SpriteRotationRate(Rotation Rate=&Multiply_Float(
     A=$Particles.InitSpriteRotRate
     B=&FloatFromCurve(
      FloatCurve=DataInterface() {propertyValues {Curve {keys [
          {arriveTangent=-1.9214, interpMode=RCIM_Cubic, leaveTangent=-1.9214, tangentMode=RCTM_User, time=-0.0046, value=1.0113}
          {arriveTangent=-0.2496, interpMode=RCIM_Cubic, leaveTangent=-0.2496, tangentMode=RCTM_User, time=0.9896, value=0.16}
         ]}}}
      CurveIndex=$Emitter.Age
     )
    ))
   &ScaleColor(
    Scale Alpha=&FloatFromCurve(FloatCurve=DataInterface() {propertyValues {Curve {keys [<at 0 0.518>,<at 0.2049 0.9765>,<at 0.852 1>,<at 1 0>]}}})
    Scale RGB=&MakeVectorFromLinearColorRGB(Color=&MaskLinearColorBySpawnGroup(
      Masked Linear Color=&ColorFromCurve(ColorCurve=DataInterface() {propertyValues {
         AlphaCurve {keys [<at 0 1 RCIM_Cubic>,<at 1 0 RCIM_Cubic>]}
         BlueCurve {keys [<at 0 0.9047 RCIM_Cubic>,<at 0.8358 0.9068 RCIM_Cubic>]}
         GreenCurve {keys [<at 0 0.6717 RCIM_Cubic>,<at 0.8358 0.5546 RCIM_Cubic>]}
         RedCurve {keys [<at 0 0.3858 RCIM_Cubic>,<at 0.8358 0.2662 RCIM_Cubic>]}
        }})
      Passthrough Linear Color=&FNTimeOfDayColor(
       Night Color=<rgba 0.1398 0.4195 1.5 1>
       Morning Color=&ColorFromCurve(ColorCurve=DataInterface() {propertyValues {
          AlphaCurve {keys [<at 0 1 RCIM_Cubic>,<at 1 0 RCIM_Cubic>]}
          BlueCurve {keys [<at 0 2 RCIM_Cubic>,<at 1 1.5 RCIM_Cubic>]}
          GreenCurve {keys [<at 0 1.2897 RCIM_Cubic>,<at 1 0.8611 RCIM_Cubic>]}
          RedCurve {keys [<at 0 1.2136 RCIM_Cubic>,<at 1 0.7443 RCIM_Cubic>]}
         }})
       Day Color=&ColorFromCurve(ColorCurve=DataInterface() {propertyValues {
          AlphaCurve {keys [<at 0 1 RCIM_Cubic>,<at 1 0 RCIM_Cubic>]}
          BlueCurve {keys [<at 0 2 RCIM_Cubic>,<at 1 1.5 RCIM_Cubic>]}
          GreenCurve {keys [<at 0 1.4933 RCIM_Cubic>,<at 1 0.9745 RCIM_Cubic>]}
          RedCurve {keys [<at 0 0.8715 RCIM_Cubic>,<at 1 0.6259 RCIM_Cubic>]}
         }})
       Evening Color=&ColorFromCurve(ColorCurve=DataInterface() {propertyValues {
          AlphaCurve {keys [<at 0 1 RCIM_Cubic>,<at 1 0 RCIM_Cubic>]}
          BlueCurve {keys [<at 0 1.647 RCIM_Cubic>,<at 1 1.0547 RCIM_Cubic>]}
          GreenCurve {keys [<at 0 1.4336 RCIM_Cubic>,<at 1 0.8676 RCIM_Cubic>]}
          RedCurve {keys [<at 0 1.0784 RCIM_Cubic>,<at 1 0.5865 RCIM_Cubic>]}
         }})
      )
     ))
    ScaleA=true
    ScaleRGB=true
   ) {module=ScaleColor001}
   &Drag(Drag=3, Use Linear Drag=true, Use Rotational Drag=false)
   &SolveForcesAndVelocity(Clamp Velocity=false, Limit Acceleration=false)
   &ScaleSpriteSize(Scale Factor=&Vector2DFromFloat(Value=&FloatFromCurve(
      FloatCurve=DataInterface() {propertyValues {Curve {keys [
          {arriveTangent=5.2354, interpMode=RCIM_Cubic, leaveTangent=5.2354, tangentMode=RCTM_User, time=-0.0046, value=1.1617}
          <at 0.1439 2.9629>
          <at 0.4209 3.1029>
          <at 1.0011 1.0032 RCIM_Cubic>
         ]}}}
      Scale Curve=&MaskFloatBySpawnGroup(
       Masked Float=1
       Passthrough Float=&FloatFromCurve(FloatCurve=DataInterface() {propertyValues {Curve {keys [
            <at 0 1>
            {
             arriveTangent=-9.3088
             arriveTangentWeight=0.3307
             interpMode=RCIM_Cubic
             leaveTangent=-9.3088
             leaveTangentWeight=0.1398
             tangentMode=RCTM_User
             time=0.7
             value=1
            }
            <at 0.9959 0 RCIM_Cubic>
           ]}}})
      )
     )))
   &DynamicMaterialParameters(
    Index 0 Param 1=&FloatFromCurve(FloatCurve=DataInterface() {propertyValues {Curve {keys [<at 0 0 RCIM_Cubic>,<at 1.0011 0.2>]}}})
    Index 0 Param 2=&FloatFromCurve(FloatCurve=DataInterface() {propertyValues {Curve {keys [
         {arriveTangent=0.0453, interpMode=RCIM_Cubic, tangentMode=RCTM_Break, time=0, value=-1}
         {arriveTangent=0.0604, interpMode=RCIM_Cubic, leaveTangent=0.0604, tangentMode=RCTM_User, time=1, value=-1}
        ]}}})
    Index 0 Param 3=&FloatFromCurve(FloatCurve=DataInterface() {propertyValues {Curve {keys [{interpMode=RCIM_Cubic, leaveTangent=4.6779, time=0.2851, value=0}]}}})
    Index 1 Param 1=&FloatFromCurve(
     FloatCurve=DataInterface() {propertyValues {Curve {keys [<at 0 0>,<at 0.25 1>]}}}
     Scale Curve=&RandomRangeFloat(Minimum=2, Maximum=1)
    )
    Index 1 Param 2=&FloatFromCurve(
     FloatCurve=DataInterface() {propertyValues {Curve {keys [
         {interpMode=RCIM_Cubic, leaveTangent=3.7422, tangentMode=RCTM_Break, time=0, value=0.2}
         {interpMode=RCIM_Cubic, leaveTangent=0.9066, time=1, value=2}
        ]}}}
     Scale Curve=&RandomRangeFloat(Minimum=0.5, Maximum=1)
    )
    Index 1 Param 3=&FloatFromCurve(FloatCurve=DataInterface() {propertyValues {Curve {keys [
         {interpMode=RCIM_Cubic, leaveTangent=2.8445, tangentMode=RCTM_Break, time=0, value=0.2}
         {interpMode=RCIM_Cubic, leaveTangent=63.2978, time=0.25, value=1}
        ]}}})
    Index 1 Param 4=&FloatFromCurve(
     FloatCurve=DataInterface() {propertyValues {Curve {keys [<at 0 1>,<at 1 0.8>]}}}
     Scale Curve=&RandomRangeFloat(Minimum=0.5, Maximum=1)
    )
    Index 2 Param 1=&FloatFromCurve(FloatCurve=DataInterface() {propertyValues {Curve {keys [{interpMode=RCIM_Cubic, leaveTangent=3.2309, tangentMode=RCTM_Break, time=0.6084, value=0}, <at 1 1>]}}})
    Index 2 Param 2=$Particles.MaterialRandom
    Index 2 Param 3=&FloatFromCurve(FloatCurve=DataInterface() {propertyValues {Curve {keys [
         {interpMode=RCIM_Cubic, tangentMode=RCTM_Break, time=0.6008, value=1}
         {interpMode=RCIM_Cubic, leaveTangent=11.1233, tangentMode=RCTM_Break, time=1, value=2}
        ]}}})
    Index 2 Param 4=0.75
    Module.Index0Float0Write=true
    Module.Index0Float1Write=true
    Module.Index0Float2Write=true
    Module.Index0Float3Write=true
    Module.Index1Float0Write=true
    Module.Index1Float1Write=true
    Module.Index1Float2Write=true
    Module.Index1Float3Write=true
    Module.Index2Float0Write=true
    Module.Index2Float1Write=true
    Module.Index2Float2Write=true
    Module.Index2Float3Write=true
    Write Parameter Index 1=::NewEnumerator1
    Write Parameter Index 2=::NewEnumerator1
   )
   &BlendSpace("Position Blend (World <-> Local)":NiagaraFloat=&MakeCustomFloatFromBool(
     Bool=&SetBoolByFloatComparison(A=&VectorLength(Vector=$Engine.Owner.Velocity), B=500)
     True Float=1
     False Float=&FloatFromCurve(
      FloatCurve=DataInterface() {propertyValues {Curve {
         defaultValue=1
         keys [
          {arriveTangent=-0.327, interpMode=RCIM_Cubic, leaveTangent=-0.327, tangentMode=RCTM_User, time=0, value=1}
          {arriveTangent=-0.1424, interpMode=RCIM_Cubic, leaveTangent=-0.1424, tangentMode=RCTM_User, time=1, value=0.3685}
         ]
        }}}
      Scale Curve=&RandomRangeFloat(Minimum=0.9, Maximum=1)
     )
     Translation Contribution:NiagaraFloat=1
     Rotation Contribution:NiagaraFloat=1
    ))
   /SetVariables(Particles.Toon Dynamic Parameters:Vector4f=&MakeVector4(
     X=&FloatFromCurve(FloatCurve=DataInterface() {propertyValues {Curve {defaultValue=1, keys [<at 0.7 0.2 RCIM_Cubic>,<at 1 1 RCIM_Cubic>]}}})
     Y=&SelectFloatFromArray(
      Array Sampling Mode=::NewEnumerator0
      Direct Array Index=?
      Float Selection Array=DataInterface() {propertyValues {FloatData [0,1,2]}}
      Random Seed=0
     )
     Z=0
     W=0
    )) {isSetParameters=true}
  ]
  rendererClasses [&NiagaraSpriteRendererProperties]
  renderers [
   {rendererClass=&NiagaraSpriteRendererProperties}
   {rendererClass=&NiagaraSpriteRendererProperties, rendererIndex=1}
  ]
 }
 Trail {
  emitterSpawnScript [/SetVariables(Emitter.WetColor:LinearColor=&MakeCustomLinearColorFromBool(Bool=$User.bIsFrontEndPreview, True Color=<rgba 0.0949 0.39 0.6383 1>, False Color=<rgba 0.097 0.1496 0.1939 0.2251>)) {isSetParameters=true}]
  emitterUpdateScript [
   &EmitterState(Life Cycle Mode=::NewEnumerator1, Loop Behavior=::NewEnumerator1, Loop Duration Mode=::NewEnumerator1, UseLoopDelay=false)
   &Timeline(
    Play=&SetBoolByFloatComparison(A=&VectorLength(Vector=$User.OwnerLocalVelocity), B=60)
    Rewind Rate=5
   )
   &SpawnRate(
    Spawn Probability=0.7
    SpawnRate=&Lerp_Float(
     A=&MakeCustomFloatFromBool(Bool=$User.bIsFrontEndPreview, True Float=2, False Float=0)
     B=&SelectFloatFromArray(Direct Array Index=$Engine.QualityLevel, Float Selection Array=DataInterface() {propertyValues {FloatData [1.5,2,3,3]}})
     Alpha=$Emitter.Timeline.NormalizedTime
    )
    Use Spawn Probability=true
   )
   &SpawnRate(
    SpawnGroup=1
    SpawnRate=&MakeCustomFloatFromBool(
     Bool=&SetBoolByFloatComparison(A=&VectorLength(Vector=$User.OwnerLocalVelocity), B=60)
     True Float=4
     False Float=&MakeCustomFloatFromBool(Bool=$User.bIsFrontEndPreview, True Float=4, False Float=0)
    )
    Use Spawn Probability=false
   ) {module=SpawnRate001}
  ]
  particleSpawnScript [
   /SetVariables(Particles.VisibilityTag=&SpawnGroup()) {isSetParameters=true}
   &InitializeParticle(
    Color=&SelectLinearColorFromArray(
     Array Sampling Mode=::NewEnumerator0
     Direct Array Index=?
     Color Selection Array=DataInterface() {propertyValues {ColorData [<rgba 1 0.9819 0.71 1>, <rgba 1 0.9623 0.4532 1>]}}
     Random Seed=0
    )
    Lifetime=2
    Position Offset=&MaskVectorBySpawnGroup(Masked Vector=<0 -130 -90>, Passthrough Vector=<0 -35 -100>)
    Position Offset Coordinate Space=::Local
    Sprite Rotation Mode=::NewEnumerator2
    Sprite Size Mode=::NewEnumerator1
    Uniform Sprite Size Max=30
    Uniform Sprite Size Min=50
    UsePositionOffset=true
    Write Material Random=true
   )
   /SetVariables(Particles.Sparkles Lifetime:NiagaraFloat=&RandomRangeFloat(Minimum=0.3, Maximum=0.4)) {isSetParameters=true}
   /SetVariables(
    Particles.DecalOrientation:Quat4f=/Game/Effects/Niagara/Modules/MergeEmitters/MergeHelper_QuatSwitch(
     0:Quat4f=<0 1 0 1>
     1:Quat4f=<0.2 1 0 1>
     2:Quat4f=<0.4 1 0 1>
     3:Quat4f=<0.6 1 0 1>
     4:Quat4f=<0.8 1 0 1>
     5:Quat4f=?
     6:Quat4f=?
     7:Quat4f=?
     8:Quat4f=?
     9:Quat4f=?
     Index:NiagaraInt32=&SelectIntFromArray(Direct Array Index=0, Int Selection Array=DataInterface() {propertyValues {IntData [0,1,2,3,4]}})
     Num:MergeHelper_SwitcherEnum=::NewEnumerator3
    )
    Particles.DecalSize:Vector3f=<200 145 145>
   ) {isSetParameters=true}
   &ShapeLocation(
    Box Size=<120 40 10>
    MaskSpawnGroup=false
    Non Uniform Scale=&MaskVector(Mask Source=0, Mask Parameter=?, Mask=1, Masked Vector=<0 0 0>, Passthrough Vector=<1 1 1>)
    Override Seed=false
    Shape Primitive=::NewEnumerator2
    Surface Only=false
    UseOwnerScale=true
   )
   &CameraOffset(Camera Offset Amount=&MaskFloatBySpawnGroup(Masked Float=20, Passthrough Float=0))
   &AddVelocity(Velocity=&MakeCustomVectorFromBool(Bool=$User.bIsFrontEndPreview, True Vector=<0 -70 0>, False Vector=<0 0 0>))
  ]
  particleUpdateScript [
   &ParticleState()
   /SetVariables(Particles.Update Sparkles Age:NiagaraFloat=&Divide_Float(A=$Particles.Age, B=$Particles.Initial.Sparkles Lifetime)) {isSetParameters=true}
   /SetVariables(Particles.DecalColor:LinearColor=&MakeLinearColorFromVectorAndFloat(
     "Vector (RGB)"=&MakeVector(
      X=&FloatFromCurve(FloatCurve=DataInterface() {propertyValues {Curve {keys [{arriveTangent=2.6346, interpMode=RCIM_Cubic, leaveTangent=2.6346, tangentMode=RCTM_User, time=0, value=-1}, <at 1 0>]}}})
      Y=0
      Z=0
     )
     "Float (Alpha)"=&FloatFromCurve(FloatCurve=DataInterface() {propertyValues {Curve {
         defaultValue=1
         keys [
          {arriveTangent=5.9515, interpMode=RCIM_Cubic, leaveTangent=5.9515, tangentMode=RCTM_User, time=0, value=0}
          {interpMode=RCIM_Cubic, tangentMode=RCTM_User, time=0.0975, value=1}
          {arriveTangentWeight=0.2201, interpMode=RCIM_Cubic, leaveTangentWeight=0.2332, tangentMode=RCTM_User, time=0.5259, value=1}
          {arriveTangent=0.086, interpMode=RCIM_Cubic, leaveTangent=0.086, tangentMode=RCTM_User, time=1, value=0}
         ]
        }}})
    )) {isSetParameters=true}
   &ScaleSpriteSize(
    Uniform Curve Index=$Particles.Update Sparkles Age
    Uniform Curve Sprite Scale=DataInterface() {propertyValues {Curve {
       defaultValue=0
       keys [
        {interpMode=RCIM_Cubic, tangentMode=RCTM_User, time=0, value=0}
        {interpMode=RCIM_Cubic, tangentMode=RCTM_User, time=0.5, value=1}
        {tangentMode=RCTM_User, time=1, value=0}
       ]
      }}}
   )
   &SpriteRotationRate(Rotation Rate=&SelectFloatFromArray(
     Array Sampling Mode=::NewEnumerator0
     Direct Array Index=?
     Float Selection Array=DataInterface() {propertyValues {FloatData [-400,400]}}
     Random Seed=0
    ))
   &BlendSpace("Position Blend (World <-> Local)":NiagaraFloat=&MaskFloatBySpawnGroup(
     Masked Float=&MakeCustomFloatFromBool(
      Bool=&SetBoolByFloatComparison(A=&VectorLength(Vector=$Engine.Owner.Velocity), B=400)
      True Float=1
      False Float=0
     )
     Passthrough Float=0
     Translation Contribution:NiagaraFloat=&MaskFloatBySpawnGroup(
      Masked Float=&MakeCustomFloatFromBool(
       Bool=&SetBoolByFloatComparison(A=&VectorLength(Vector=$Engine.Owner.Velocity), B=400)
       True Float=1
       False Float=0
      )
      Passthrough Float=0
     )
     Rotation Contribution:NiagaraFloat=&MaskFloatBySpawnGroup(
      Masked Float=&MakeCustomFloatFromBool(
       Bool=&SetBoolByFloatComparison(A=&VectorLength(Vector=$Engine.Owner.Velocity), B=400)
       True Float=1
       False Float=0
      )
      Passthrough Float=0
     )
    ))
   &SolveForcesAndVelocity(Clamp Velocity=false, Limit Acceleration=false)
  ]
  rendererClasses [&NiagaraSpriteRendererProperties, &NiagaraDecalRendererProperties]
  renderers [
   {rendererClass=&NiagaraSpriteRendererProperties}
   {rendererClass=&NiagaraDecalRendererProperties, rendererIndex=1}
  ]
 }
 TrailBubbles {
  simTarget=GPUComputeSim
  emitterUpdateScript [
   &EmitterState(Life Cycle Mode=::NewEnumerator1, Loop Delay=0.5, Loop Duration=1, UseLoopDelay=false)
   &SpawnRate(
    SpawnRate=&MakeCustomFloatFromBool(
     Bool=&SetBoolByFloatComparison(A=&VectorLength(Vector=$User.OwnerLocalVelocity), B=60)
     True Float=70
     False Float=&MakeCustomFloatFromBool(Bool=$User.bIsFrontEndPreview, True Float=70, False Float=0)
    )
    Use Spawn Probability=true
   )
  ]
  particleSpawnScript [
   &PartitionParticles()
   &InitializeParticle(
    Lifetime=&RandomRangeFloat(Minimum=0.8, Maximum=1)
    Sprite Size Mode=::NewEnumerator1
    Uniform Sprite Size Max=2
    Uniform Sprite Size Min=0.5
    UsePositionOffset=false
    Write Material Random=true
   )
   &ApplyInitialForces()
   &SocketLocation(
    SamplePosOffset=true
    Socket Index=&SelectIntFromArray(Direct Array Index=$Particles.PartitionParticles.Partition, Int Selection Array=DataInterface() {propertyValues {IntData [0,1]}})
    Socket Reader=$System.Sweeper Socket Reader
    Socket Sampling Mode=::NewEnumerator3
    VelocitySpeedLimit=false
   )
   &ShapeLocation(
    Handle Radius=0
    Large Radius=20
    MaskSpawnGroup=false
    Non Uniform Scale=<1 1 0>
    Offset=&MaskVector(Mask Source=1, Mask Parameter=$Particles.PartitionParticles.Partition, Mask=0, Masked Vector=<3 0 -10>, Passthrough Vector=<-3 0 -10>)
    Offset Mode=::NewEnumerator0
    Override Seed=false
    Rotation Mode=::NewEnumerator2
    Shape Origin=$Output.SocketLocation.SampledPosition
    Shape Primitive=::NewEnumerator3
    U Distribution=0.5
    UseOwnerScale=true
    Yaw / Pitch / Roll=&MaskVector(Mask Source=1, Mask Parameter=$Particles.PartitionParticles.Partition, Mask=0, Masked Vector=<0.1 0 0>, Passthrough Vector=<-0.1 0 0>)
   )
   &AddVelocity(Velocity=&RandomRangeVector(Minimum=<0 0 10>, Maximum=<0 0 0>)) {module=AddVelocity001}
   &AddVelocity(Velocity=&MakeCustomVectorFromBool(
     Bool=$User.bIsFrontEndPreview
     True Vector=&RandomRangeVector(Minimum=<0 -70 0>, Maximum=<0 -150 0>)
     False Vector=<0 0 0>
    ))
  ]
  particleUpdateScript [
   &ParticleState()
   &GravityForce(Gravity=&RandomRangeVector(
     Minimum=&VectorFromCurve(VectorCurve=DataInterface() {propertyValues {
        XCurve {keys [<at 0 0>,<at 1 0>]}
        YCurve {keys [<at 0 0>,<at 1 0>]}
        ZCurve {keys [
          {interpMode=RCIM_Cubic, tangentMode=RCTM_User, time=0, value=10}
          {arriveTangent=88.4209, interpMode=RCIM_Cubic, leaveTangent=88.4209, tangentMode=RCTM_User, time=0.75, value=35}
         ]}
       }})
     Maximum=&VectorFromCurve(VectorCurve=DataInterface() {propertyValues {
        XCurve {keys [<at 0 0>,<at 1 0>]}
        YCurve {keys [<at 0 0>,<at 1 0>]}
        ZCurve {keys [
          {interpMode=RCIM_Cubic, tangentMode=RCTM_User, time=0, value=10}
          {arriveTangent=133.5137, interpMode=RCIM_Cubic, leaveTangent=133.5137, tangentMode=RCTM_User, time=0.75, value=70}
         ]}
       }})
    ))
   &DynamicMaterialParameters(
    Index 0 Param 1=&FloatFromCurve(FloatCurve=DataInterface() {propertyValues {Curve {keys [
         {arriveTangent=0.1706, interpMode=RCIM_Cubic, leaveTangent=0.1706, leaveTangentWeight=0.3081, tangentMode=RCTM_User, time=0, value=0}
         {arriveTangentWeight=0.3104, interpMode=RCIM_Cubic, leaveTangentWeight=0.0345, tangentMode=RCTM_User, time=0.9, value=0.2}
         {arriveTangent=3.2005, interpMode=RCIM_Cubic, leaveTangent=3.2005, tangentMode=RCTM_User, time=1, value=0.5}
        ]}}})
    Module.Index0Float0Write=true
    Module.Index0Float1Write=true
    Module.Index0Float2Write=true
    Module.Index0Float3Write=true
   )
   &ScaleColor(
    Scale Alpha=&FloatFromCurve(
     FloatCurve=DataInterface() {propertyValues {Curve {
        defaultValue=0.3
        keys [
         {tangentMode=RCTM_User, time=0, value=1}
         {arriveTangent=-0.0405, interpMode=RCIM_Cubic, leaveTangent=-0.0405, tangentMode=RCTM_User, time=0.7009, value=1}
         {arriveTangent=-8.0089, interpMode=RCIM_Cubic, leaveTangent=-8.0089, tangentMode=RCTM_User, time=1, value=0}
        ]
       }}}
     Scale Curve=&FNTimeOfDayScalar(Night Scale=0.1, Morning Scale=0.3, Day Scale=0.5, Evening Scale=0.3)
    )
    ScaleA=true
    ScaleRGB=true
   )
   &ParticleState() {module=ParticleState001}
   &ScaleSpriteSize(Scale Factor=&Vector2DFromFloat(Value=&FloatFromCurve(
      FloatCurve=DataInterface() {propertyValues {Curve {keys [
          {arriveTangent=4.1785, interpMode=RCIM_Cubic, leaveTangent=4.1785, tangentMode=RCTM_User, time=0, value=0}
          {arriveTangent=1.8588, interpMode=RCIM_Cubic, leaveTangent=1.8588, tangentMode=RCTM_User, time=0.1, value=0.35}
          {arriveTangent=-0.2915, interpMode=RCIM_Cubic, leaveTangent=-0.2915, tangentMode=RCTM_User, time=0.6, value=1}
          {arriveTangentWeight=0.199, interpMode=RCIM_Cubic, leaveTangentWeight=0.0479, tangentMode=RCTM_User, time=0.9719, value=1}
          <at 1 2 RCIM_Cubic>
         ]}}}
      Scale Curve=4
     )))
   &SpriteRotationRate(Rotation Rate=&RandomRangeFloat(Minimum=-40, Maximum=40))
   &Drag(
    Drag=&FloatFromCurve(
     FloatCurve=DataInterface() {propertyValues {Curve {
        defaultValue=1
        keys [
         {interpMode=RCIM_Cubic, tangentMode=RCTM_User, time=0, value=0.2547}
         {interpMode=RCIM_Cubic, tangentMode=RCTM_User, time=0.2421, value=0.8074}
        ]
       }}}
     Scale Curve=&RandomRangeFloat(Minimum=3, Maximum=5)
    )
    Use Linear Drag=true
    Use Rotational Drag=true
   )
   &SolveForcesAndVelocity(Clamp Velocity=false, Limit Acceleration=false)
   &BlendSpace("Position Blend (World <-> Local)":NiagaraFloat=&MakeCustomFloatFromBool(
     Bool=&SetBoolByFloatComparison(A=&VectorLength(Vector=$Engine.Owner.Velocity), B=500)
     True Float=1
     False Float=&FloatFromCurve(FloatCurve=DataInterface() {propertyValues {Curve {
         defaultValue=1
         keys [{interpMode=RCIM_Cubic, tangentMode=RCTM_User, time=0, value=0.6381}, {interpMode=RCIM_Cubic, tangentMode=RCTM_User, time=1, value=0.3685}]
        }}})
     Translation Contribution:NiagaraFloat=1
     Rotation Contribution:NiagaraFloat=&MakeCustomFloatFromBool(
      Bool=&SetBoolByFloatComparison(A=&VectorLength(Vector=$Engine.Owner.Velocity), B=200)
      True Float=1
      False Float=&FloatFromCurve(FloatCurve=DataInterface() {propertyValues {Curve {
          defaultValue=1
          keys [{interpMode=RCIM_Cubic, tangentMode=RCTM_User, time=0, value=0.4314}, {interpMode=RCIM_Cubic, tangentMode=RCTM_User, time=1, value=0}]
         }}})
     )
    ))
  ]
  rendererClasses [&NiagaraSpriteRendererProperties]
  renderers [{rendererClass=&NiagaraSpriteRendererProperties}]
 }
 WetDecal {
  bEnabled=false
  emitterSpawnScript [/SetVariables(Emitter.WetColor:LinearColor=&MakeCustomLinearColorFromBool(Bool=$User.bIsFrontEndPreview, True Color=<rgba 0.0949 0.39 0.6383 1>, False Color=<rgba 0.097 0.1496 0.1939 0.2251>)) {isSetParameters=true}]
  emitterUpdateScript [
   &EmitterState(Life Cycle Mode=::NewEnumerator1, Loop Behavior=::NewEnumerator1, Loop Duration Mode=::NewEnumerator1, UseLoopDelay=false)
   &SpawnRate(
    SpawnRate=&MakeCustomFloatFromBool(
     Bool=&SetBoolByFloatComparison(A=&VectorLength(Vector=$User.OwnerLocalVelocity), B=60)
     True Float=4
     False Float=&MakeCustomFloatFromBool(Bool=$User.bIsFrontEndPreview, True Float=4, False Float=0)
    )
    Use Spawn Probability=false
   )
  ]
  particleSpawnScript [
   &InitializeParticle(
    Color Mode=::NewEnumerator0
    Lifetime=2
    Position Offset=<0 -35 -100>
    Position Offset Coordinate Space=::Local
    UsePositionOffset=true
    Write Material Random=true
   )
   /SetVariables(
    Particles.DecalOrientation:Quat4f=&MergeHelper_QuatSwitch(
     0:Quat4f=<0 1 0 1>
     1:Quat4f=<0.2 1 0 1>
     2:Quat4f=<0.4 1 0 1>
     3:Quat4f=<0.6 1 0 1>
     4:Quat4f=<0.8 1 0 1>
     5:Quat4f=?
     6:Quat4f=?
     7:Quat4f=?
     8:Quat4f=?
     9:Quat4f=?
     Index:NiagaraInt32=&SelectIntFromArray(Direct Array Index=0, Int Selection Array=DataInterface() {propertyValues {IntData [0,1,2,3,4]}})
     Num:MergeHelper_SwitcherEnum=::NewEnumerator3
    )
    Particles.DecalSize:Vector3f=<200 145 145>
   ) {isSetParameters=true}
   &AddVelocity(Velocity=&MakeCustomVectorFromBool(Bool=$User.bIsFrontEndPreview, True Vector=<0 -70 0>, False Vector=<0 0 0>))
  ]
  particleUpdateScript [
   &ParticleState()
   &ScaleColor(
    Scale Alpha=&FloatFromCurve(FloatCurve=DataInterface() {propertyValues {Curve {
        defaultValue=1
        keys [
         {arriveTangent=5.9515, interpMode=RCIM_Cubic, leaveTangent=5.9515, tangentMode=RCTM_User, time=0, value=0}
         {interpMode=RCIM_Cubic, tangentMode=RCTM_User, time=0.0975, value=1}
         {arriveTangentWeight=0.2201, interpMode=RCIM_Cubic, leaveTangentWeight=0.2332, tangentMode=RCTM_User, time=0.5259, value=1}
         {arriveTangent=0.086, interpMode=RCIM_Cubic, leaveTangent=0.086, tangentMode=RCTM_User, time=1, value=0}
        ]
       }}})
    Scale RGB=&MakeVector(
     X=&FloatFromCurve(FloatCurve=DataInterface() {propertyValues {Curve {keys [{arriveTangent=2.6346, interpMode=RCIM_Cubic, leaveTangent=2.6346, tangentMode=RCTM_User, time=0, value=-1}, <at 1 0>]}}})
     Y=0
     Z=0
    )
    ScaleA=true
    ScaleRGB=true
   )
   &SolveForcesAndVelocity(Clamp Velocity=false, Limit Acceleration=false)
  ]
  rendererClasses [&NiagaraDecalRendererProperties]
  renderers [{rendererClass=&NiagaraDecalRendererProperties, rendererIndex=0}]
 }
}
summary {
 systemName=NS_Emote_SweepingClean_Bubbles
 emitters [
  {emitterName=Bubbles, rendererClasses [&NiagaraSpriteRendererProperties]}
  {emitterName=Mist, rendererClasses [&NiagaraSpriteRendererProperties]}
  {bEnabled=false, emitterName=WetDecal, rendererClasses [&NiagaraDecalRendererProperties]}
  {emitterName=Trail, rendererClasses [&NiagaraSpriteRendererProperties, &NiagaraDecalRendererProperties]}
  {emitterName=Soap, rendererClasses [&NiagaraSpriteRendererProperties]}
  {emitterName=BurstStaticBubbles, rendererClasses [&NiagaraSpriteRendererProperties]}
  {emitterName=TrailBubbles, simTarget=GPUComputeSim, rendererClasses [&NiagaraSpriteRendererProperties]}
  {emitterName=Suds, rendererClasses [&NiagaraSpriteRendererProperties]}
 ]
 userVariables [
  User.bIsFrontEnd {type=NiagaraBool}
  User.bIsFrontEndPreview {type=NiagaraBool}
  User.OwnerLocalVelocity {type=Vector3f, description="", defaultValue {struct=Vector3f, value=<0 0 0>}}
 ]
}
